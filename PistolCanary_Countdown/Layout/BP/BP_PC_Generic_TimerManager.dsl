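; Timer manager Blueprint for the PistolCanary Countdown content that switches on World Partition data layers when countdown events fire.
; Its overrides of EventContinueClockCountdown and the PC_Countdown custom event each activate three Day1 data layers (End_S3, MatchEnd, End_S0) non-recursively, then call the parent implementation.
; Both events duplicate the same hardcoded activation, which author comments mark as a quick, temporary fix.
; The construction script only calls its parent.
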(event EventContinueClockCountdown
  (bind _returnvalue (WorldPartition|GetDataLayerManager))
  ;[0.15 0 0.000109 0.5] Quick and Dirty - TODO: Need to activate these data layers in a more intentional way.
  ; ..just getting this working tonight (5/29/25 -Adam)
  (DataLayers|SetDataLayerRuntimeState _returnvalue "/PistolCanary_Countdown/DataLayers/DL_PistolCanary_Day1_to_End_S3.DL_PistolCanary_Day1_to_End_S3" "Activated" false :!class /Script/Engine.DataLayerManager)
  (DataLayers|SetDataLayerRuntimeState _returnvalue "/PistolCanary_Countdown/DataLayers/DL_PistolCanary_Day1_to_MatchEnd.DL_PistolCanary_Day1_to_MatchEnd" "Activated" false :!class /Script/Engine.DataLayerManager)
  (DataLayers|SetDataLayerRuntimeState _returnvalue "/PistolCanary_Countdown/DataLayers/DL_PistolCanary_Day1_to_End_S0.DL_PistolCanary_Day1_to_End_S0" "Activated" false :!class /Script/Engine.DataLayerManager)
  (|CallParentFunction))

(event Custom|PC_Countdown
  (bind _returnvalue (WorldPartition|GetDataLayerManager))
  ;[0.15 0 0.000109 0.5] Quick and Dirty - TODO: Need to activate these data layers in a more intentional way.
  ; ..just getting this working tonight (5/29/25 -Adam)
  (DataLayers|SetDataLayerRuntimeState _returnvalue "/PistolCanary_Countdown/DataLayers/DL_PistolCanary_Day1_to_End_S3.DL_PistolCanary_Day1_to_End_S3" "Activated" false :!class /Script/Engine.DataLayerManager)
  (DataLayers|SetDataLayerRuntimeState _returnvalue "/PistolCanary_Countdown/DataLayers/DL_PistolCanary_Day1_to_MatchEnd.DL_PistolCanary_Day1_to_MatchEnd" "Activated" false :!class /Script/Engine.DataLayerManager)
  (DataLayers|SetDataLayerRuntimeState _returnvalue "/PistolCanary_Countdown/DataLayers/DL_PistolCanary_Day1_to_End_S0.DL_PistolCanary_Day1_to_End_S0" "Activated" false :!class /Script/Engine.DataLayerManager)
  (|CallParentFunction))

(fn ConstructionScript
  (|CallParentFunction))
